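# Lit meteor material for mesh particles, with tinted rock, blended normals, hot emissive ends and blue crystal cracks.
# A world-space gradient along GradientDirection, scaled by object radius, splits front and rear tint and glow, while vertex-color blue masks the cracks.
# Particle color scales that emissive before a portal glow is added near a parameter-defined plane.
# The same plane test builds an offset pulling points on one side to PlaneLocation.
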Material /Figment_S10/Athena/Prototype/Blueprints/NightNight/FX/Meteor/M_Meteor_Particle
graph {
 properties {bUsedWithMeshParticles=true, pixelDepthOffsetMode=PDOM_Legacy, usedShadingModels=Default Lit}
 nodes [
  n1 = TextureSampleParameter2D() {
   parameterName=Mask
   samplerSource=SSM_Wrap_WorldGroupSettings
   samplerType=SAMPLERTYPE_Masks
   texture=/Game/Athena/Environments/Meteor/Textures/Meteor_M
  }
  n2 = LinearInterpolate(
   A=LinearInterpolate(
    A=VectorParameter() {parameterName=Color1, defaultValue=<rgba 0.7656 0.8882 1 1>}.RGB
    B=VectorParameter() {parameterName=Color2, defaultValue=<rgba 0.8197 0.7135 1 1>}.RGB
    Alpha=@n1.G
   )
   B=VectorParameter() {parameterName=Color3, defaultValue=<rgba 1 0.9075 0.7812 1>}.RGB
   Alpha=@n1.B
  )
  n3 = LinearInterpolate(
   A=@n2
   B=Multiply(A=@n2, B=ScalarParameter() {defaultValue=2, parameterName=FresnelBrightness})
   Alpha=Fresnel(ExponentIn=ScalarParameter() {defaultValue=4, parameterName=FresnelMeteor}, Normal=VertexNormalWS()) {baseReflectFraction=0}
  )
  n4 = Multiply(A=TextureCoordinate(), B=ScalarParameter() {defaultValue=6, parameterName=SizeMeteor_Tiling})
  n5 = Multiply(
   A=MaterialProxyReplace(Realtime=ShadingPathSwitch(Default=@n3, Deferred=@n3, Forward=@n2, Mobile=@n2), MaterialProxy=@n2)
   B=TextureSample(UVs=@n4) {texture=/Game/Athena/Environments/Meteor/Textures/T_Meteor_01_C}.RGB
  )
  n6 = TextureSampleParameter2D() {
   parameterName=Mask
   samplerSource=SSM_Wrap_WorldGroupSettings
   samplerType=SAMPLERTYPE_Masks
   texture=&Meteor_M
  }
  n7 = Multiply(A=ComponentMask(None=@n5) {r=true}, B=@n6.R)
  n8 = Multiply(
   A=Power(Base=OneMinus(None=@n7), Exp=ScalarParameter() {defaultValue=37.4447, parameterName=AOMaskPower})
   B=ScalarParameter() {defaultValue=1.9946, parameterName=AOMaskBrightness}
  )
  n9 = Clamp(None=LinearInterpolate(
    A=ScalarParameter() {defaultValue=-22.699, parameterName=Cracks_Low}
    B=ScalarParameter() {defaultValue=269.5855, parameterName=Cracks_high}
    Alpha=ComponentMask(None=@n5) {r=true}
   )) {clampMode=CMODE_ClampMin}
  n10 = Add(A=@n8, B=@n9)
  n11 = VectorParameter() {parameterName=GradientDirection, defaultValue=<rgba 0 0 1 1>}
  n12 = DotProduct(A=Normalize(VectorInput=@n11.RGB), B=Subtract(A=WorldPosition().XYZ, B=ObjectPositionWS()))
  n13 = Multiply(A=ObjectRadius(), B=ScalarParameter() {defaultValue=2, parameterName=Gradient})
  n14 = Clamp(None=Add(A=Divide(A=@n12, B=@n13), B=ScalarParameter() {defaultValue=0.6, parameterName=Gradientoffset}) {constB=0.5})
  n15 = Power(Base=OneMinus(None=@n14), Exp=ScalarParameter() {defaultValue=7, parameterName=GradientFalloffRear})
  n16 = Add(A=OneMinus(None=@n8), B=@n9)
  n17 = Power(Base=@n14, Exp=ScalarParameter() {defaultValue=7, parameterName=GradientFalloffFront})
  n18 = Add(
   A=Multiply(A=Multiply(A=@n10, B=@n15), B=VectorParameter() {parameterName=EmissiveRear, defaultValue=<rgba 20 4.3722 0 1>}.RGB)
   B=Multiply(A=Multiply(A=@n16, B=@n17), B=VectorParameter() {parameterName=EmissiveFront, defaultValue=<rgba 10 1.3829 0 1>}.RGB)
  )
  n19 = VertexColor()
  n20 = Multiply(A=@n8, B=ScalarParameter() {defaultValue=0.01, parameterName=CracksBrightness2})
  n21 = TextureSample(UVs=Multiply(A=TextureCoordinate(), B=ScalarParameter() {defaultValue=12, parameterName=CrackSize})) {texture=/Game/Athena/Environments/Meteor/Textures/Meteor_Cracks_M}
  n22 = ComponentMask(None=@n21.RGB) {b=true}
  n23 = ScalarParameter() {defaultValue=5, parameterName=CracksPower}
  n24 = Multiply(A=@n8, B=Multiply(A=ScalarParameter() {defaultValue=50, parameterName=CracksBrightness}, B=Power(Base=@n22, Exp=@n23)))
  n25 = Multiply(
   A=VectorParameter() {parameterName=EmissiveCracksBlue, defaultValue=<rgba 0.604 1.0675 5 1>}.RGB
   B=Multiply(A=@n19.B, B=Add(A=@n20, B=@n24))
  )
  n26 = VectorParameter() {parameterName=PlaneLocation, defaultValue=<rgba 65949.1797 -42629 248989 0>}
  n27 = WorldPosition()
  n28 = DotProduct(
   A=Normalize(VectorInput=Subtract(A=@n26.RGB, B=@n27.XYZ))
   B=VectorParameter() {parameterName=MeteorDirection, defaultValue=<rgba 0 0 1 0>}.RGB
  )
  n29 = Saturate(None=OneMinus(None=Abs(None=@n28)))
  n30 = ScalarParameter() {defaultValue=4, parameterName=PortalMaskPower}
  Add(
   A=Multiply(A=ParticleColor().RGB, B=Add(A=@n18, B=@n25))
   B=Multiply(A=VectorParameter() {parameterName=PortalColor, defaultValue=<rgba 0.3125 14.647 20 0>}.RGB, B=Power(Base=@n29, Exp=@n30) {constExponent=4})
  )
  Constant()
  LinearInterpolate(A=Subtract(A=@n26.RGB, B=WorldPosition().XYZ), Alpha=Ceil(None=Saturate(None=@n28))) {constB=0}
  n31 = ComponentMask(None=@n21.RGB) {g=true}
  n32 = ScalarParameter() {parameterName=CrackShadow_Brightness}
  n33 = ScalarParameter() {defaultValue=8, parameterName=CrackshadowBias}
  n34 = LinearInterpolate(A=Constant() {r=1}, B=Clamp(None=Add(A=@n31, B=@n32)), Alpha=Clamp(None=Multiply(A=@n19.B, B=@n33)))
  n35 = LinearInterpolate(
   A=VectorParameter() {parameterName=TintRear, defaultValue=<rgba 0.01 0.0006 0.0004 1>}.RGB
   B=VectorParameter() {parameterName=TintFornt, defaultValue=<rgba 1.5 1.5 1.5 1>}.RGB
   Alpha=@n14
  )
  n36 = Multiply(A=@n22) {constA=3, constB=4}
  LinearInterpolate(
   A=Multiply(A=@n5, B=Multiply(A=@n35, B=@n34))
   B=VectorParameter() {parameterName=CrystalColor, defaultValue=<rgba 0.0136 0.0169 0.1823 1>}.RGB
   Alpha=Multiply(A=@n19.B, B=Clamp(None=@n36))
  )
  Multiply(
   A=MaterialFunctionCall(
    BaseNormal=TextureSampleParameter2D() {parameterName=Normal, samplerType=SAMPLERTYPE_Normal, texture=/Game/Athena/Environments/Meteor/Textures/Meteor_N}.RGB
    AdditionalNormal=TextureSample(UVs=@n4) {samplerType=SAMPLERTYPE_Normal, texture=/Game/Athena/Environments/Meteor/Textures/T_Meteor_01_N}.RGB
   ) {materialFunction=&BlendAngleCorrectedNormals}.Result
   B=VectorParameter() {parameterName=NormalMult, defaultValue=<rgba 1.2 1.2 1 1>}.RGB
  )
  Transform(None=@n11.RGB) {transformSourceType=TRANSFORMSOURCE_Local}
 ]
}
