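# Masked, lit rock material for the NightNight prototype meteor, apparently cut and lit by a portal plane.
# The dot product of MeteorDirection and the normalized pixel-to-PlaneLocation vector, saturated and ceiled, gives a binary one-side mask.
# The inverted absolute dot, sharpened by PortalMaskPower, scales HDR PortalColor into a glow band along that plane.
# Surface detail reuses Pirate Cove diffuse, darkened by a constant, plus Arid rock packed and normal textures.
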Material /Figment_S10/Athena/Prototype/Blueprints/NightNight/FX/Meteor/M_Meteor_Rock
graph {
 properties {bUsedWithNanite=true, blendMode=BLEND_Masked, pixelDepthOffsetMode=PDOM_Legacy, usedShadingModels=Default Lit}
 nodes [
  n1 = VectorParameter() {parameterName=PlaneLocation, defaultValue=<rgba 24056 2399 4844 0>}
  n2 = WorldPosition()
  n3 = DotProduct(
   A=Normalize(VectorInput=Subtract(A=@n1.RGB, B=@n2.XYZ))
   B=VectorParameter() {parameterName=MeteorDirection, defaultValue=<rgba 0.7755 -0.4565 -0.4361 0>}.RGB
  )
  Ceil(None=Saturate(None=@n3))
  n4 = TextureSample() {texture=/Game/Environments/Sets/Arid/Textures/Arid_Rocks/T_Rocks_S}
  LinearInterpolate(Alpha=@n4.B) {constA=0.4, constB=0.7}
  Multiply(
   A=Constant3Vector() {constant=<rgba 0.0677 0.0677 0.0677 1>}
   B=TextureSample() {texture=/Game/Environments/Sets/Pirate_Cove/Materials/T_Rock02_TP_D}.RGB
  )
  Multiply(A=@n4.R) {constB=2}
  n5 = OneMinus(None=Abs(None=@n3))
  Multiply(
   A=VectorParameter() {parameterName=PortalColor, defaultValue=<rgba 0.3125 14.647 20 0>}.RGB
   B=Power(Base=Saturate(None=@n5), Exp=ScalarParameter() {defaultValue=8, parameterName=PortalMaskPower}) {constExponent=4}
  )
  Subtract(A=@n1.RGB, B=WorldPosition().XYZ)
  TextureSample() {samplerType=SAMPLERTYPE_Normal, texture=/Game/Environments/Sets/Arid/Textures/Arid_Rocks/T_Rocks_N}
 ]
}
